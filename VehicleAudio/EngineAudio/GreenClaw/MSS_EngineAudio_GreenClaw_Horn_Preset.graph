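# MetaSound Source preset that plays the exterior horn sound for the GreenClaw vehicle.
# It inherits its graph from MSS_Vehicle_Horn and supplies the GreenClaw horn wave and an output volume as inputs.
# The source is a mono one-shot with an OnFinished trigger, using the shared vehicle horn attenuation, cosmetic horn property sheet and exterior vehicle sound class.
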MetaSoundSource /VehicleAudio/EngineAudio/GreenClaw/MSS_EngineAudio_GreenClaw_Horn_Preset
preset_parent=/VehicleAudio/Patches/MSS_Vehicle_Horn
inputs [
 OutputVolume {type=Float, access=Constructor}
 UE.Source.OnPlay {type=Trigger}
 Wave Asset {type=WaveAsset}
]
interfaces [UE.OutputFormat.Mono:1.0, UE.Source.OneShot:1.0, UE.Source:1.1]
outputs [
 "UE.OutputFormat.Mono.Audio:0" {type=Audio}
 UE.Source.OneShot.OnFinished {type=Trigger}
]
pages [Default {input_defaults {OutputVolume=9, UE.Source.OnPlay=null, Wave Asset=&/EngineAudio_GreenClaw_Horn}}]
properties {
 assetImportData=&self:AssetImportData
 attenuationSettings=/VehicleAudio/Settings/Vehicle_Horn_Att
 audioPropertiesSheet=/VehicleAudio/PropertySheets/APS_Cosmetic_Vehicle_Horn
 maxDistance=16030.001
 sampleRate=48000
 soundClassObject=&Fort_Vehicles_Exterior
 volume=1.585
}
